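# Level Sequence holding a facial animation for the AlpacaLean_Gel custom skin, filed under Battle Royale menu player animations.
# It binds a FortPlayerMannequin with a spawn track, a slightly scaled-down transform and an empty skeletal animation track.
# A Fortnite_3L_Face_CtrlRig Control Rig track carries the performance over roughly 440 frames.
# Densely keyed eye aim, blink, lid, squint and brow controls animate glances and expressions.
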LevelSequence /BRCosmetics/Animation/Game/MainPlayer/Menu/BR/CustomSkins/_S41/AlpacaLean_Gel/AlpacaGel_Anim/AlpacaLean_Gel_anim
bindings [FortPlayerMannequin {
  id=#1
  possessed_class=FortPlayerMannequin
  tracks [
   {type=MovieSceneSpawnTrack, sections [{type=MovieSceneSpawnSection, channels [None {type=MovieSceneScriptingBoolChannel, default=true}]}]}
   {
    type=MovieScene3DTransformTrack
    sections [{
      type=MovieScene3DTransformSection
      channels [
       Location.X {type=MovieSceneScriptingDoubleChannel, default=0}
       Location.Y {type=MovieSceneScriptingDoubleChannel, default=0}
       Location.Z {type=MovieSceneScriptingDoubleChannel, default=0}
       Rotation.X {type=MovieSceneScriptingDoubleChannel, default=0}
       Rotation.Y {type=MovieSceneScriptingDoubleChannel, default=0}
       Rotation.Z {type=MovieSceneScriptingDoubleChannel, default=0}
       Scale.X {type=MovieSceneScriptingDoubleChannel, default=0.9}
       Scale.Y {type=MovieSceneScriptingDoubleChannel, default=0.9}
       Scale.Z {type=MovieSceneScriptingDoubleChannel, default=0.9}
      ]
     }]
   }
   {type=MovieSceneSkeletalAnimationTrack, sections []}
   {
    type=MovieSceneControlRigParameterTrack
    display_name=Fortnite_3L_Face_CtrlRig
    sections [{
      type=MovieSceneControlRigParameterSection
      channels [
       CTRL_C_eyesAim.Location.X {
        type=MovieSceneScriptingFloatChannel
        default=0
        thinned_from=440
        keys [[0,0],[31,0],[32,-8.3623],[33,-15.6865],[65,-15.6865],[66,-12.9094],[67,-10.5734],[68,-7.9538],[69,-2.8276],[70,0],[143,0],[144,4.3397],[145,8.6156],[146,9.0242],[147,9.3196],[148,9.5767],[150,9.8873],[151,9.8327],[152,9.7049],[153,9.6134],[155,9.5401],[165,9.2459],[166,9.2275],[168,9.1536],[169,9.1352],[173,8.9872],[174,8.9687],[179,8.7831],[180,8.7645],[183,8.6528],[184,8.6342],[185,8.5969],[186,8.5783],[187,8.541],[188,8.5223],[189,8.485],[192,8.4289],[193,8.3915],[194,8.3728],[195,8.3728],[198,8.3166],[199,8.3166],[200,8.2979],[360,8.2979],[361,8.1103],[362,7.601],[363,6.8208],[364,5.859],[365,4.7702],[366,3.6518],[367,2.5454],[368,1.5529],[369,0.737],[370,0.1992],[371,0],[439,0]]
       }
       CTRL_C_eyesAim.Location.Y {
        type=MovieSceneScriptingFloatChannel
        default=0
        thinned_from=440
        keys [[0,0.7814],[31,0.7814],[32,1.7646],[33,5.6893],[34,6.0308],[35,7.1273],[36,7.9502],[44,7.9502],[45,8.9577],[62,8.9577],[63,6.043],[64,6.1679],[65,6.4945],[66,4.926],[67,3.8936],[68,2.6386],[69,1.1294],[70,0.7814],[142,0.7814],[143,-0.0797],[144,0.4919],[145,1.6709],[146,1.7849],[147,1.8514],[148,1.9117],[150,1.9699],[151,1.9004],[152,1.9131],[153,1.9835],[154,2.0003],[155,2.0078],[158,2.0526],[159,2.0804],[160,2.0986],[161,2.1063],[164,2.1123],[165,2.1095],[167,2.1166],[168,2.1146],[169,2.1247],[170,2.1277],[173,2.1532],[174,2.1798],[175,2.1951],[177,2.2438],[179,2.297],[180,2.3352],[181,2.3617],[182,2.3832],[184,2.4417],[185,2.4655],[186,2.5034],[187,2.5355],[188,2.5753],[189,2.6092],[190,2.6509],[191,2.7007],[193,2.7749],[195,2.8657],[199,2.9939],[202,3.0435],[204,3.0435],[208,3.0113],[211,2.9635],[216,2.8469],[222,2.6689],[223,2.6331],[224,2.6048],[225,2.5699],[226,2.5423],[228,2.4745],[229,2.4479],[232,2.3507],[233,2.3255],[234,2.2944],[235,2.2699],[236,2.2397],[242,2.1009],[247,2.0197],[250,1.9884],[254,1.9679],[257,1.9679],[258,1.973],[259,1.973],[260,1.9781],[261,1.9781],[268,2.0144],[269,2.025],[270,2.0303],[271,2.041],[272,2.0463],[273,2.0571],[274,2.0625],[275,2.0734],[276,2.0789],[278,2.1009],[279,2.1065],[281,2.129],[282,2.1346],[284,2.1574],[285,2.1632],[286,2.1747],[287,2.1806],[288,2.1922],[289,2.1981],[290,2.2099],[299,2.2638],[300,2.2638],[301,2.2699],[319,2.2699],[320,1.494],[321,1.3476],[322,1.5787],[323,1.9165],[324,2.2396],[325,2.3755],[360,2.3755],[361,2.217],[362,1.8193],[364,0.7671],[365,0.3527],[366,0.1254],[367,0.1006],[368,0.2445],[369,0.4794],[370,0.6922],[371,0.7814],[439,0.7814]]
       }
       CTRL_C_eyesAim.Location.Z {
        type=MovieSceneScriptingFloatChannel
        default=0
        thinned_from=440
        keys [[0,-3.9669],[31,-3.9669],[32,-1.795],[33,0.1406],[34,-1.6643],[35,-4.9048],[36,-6.5431],[44,-6.5431],[45,-8.1363],[62,-8.1363],[63,-1.7147],[64,-2.2006],[65,-3.2766],[66,-4.8065],[67,-5.657],[68,-5.3834],[69,-4.5604],[70,-3.9669],[142,-3.9669],[143,0.8371],[144,6.2799],[145,6.2223],[146,5.9156],[147,5.5085],[148,5.0884],[149,4.6414],[150,4.1789],[151,2.8311],[152,0.9935],[153,-0.0821],[154,-0.3286],[156,-0.6571],[158,-1.0269],[159,-1.2319],[160,-1.4167],[162,-1.5815],[163,-1.6433],[164,-1.7261],[166,-1.8504],[167,-1.9333],[168,-1.9958],[170,-2.1611],[173,-2.4715],[174,-2.6148],[175,-2.7386],[176,-2.903],[177,-3.0471],[178,-3.2113],[179,-3.3552],[180,-3.5181],[181,-3.6617],[182,-3.7851],[183,-3.9286],[185,-4.1737],[190,-4.8837],[191,-5.0443],[192,-5.165],[194,-5.4471],[195,-5.5656],[198,-5.8659],[199,-5.9444],[200,-6.005],[202,-6.0833],[204,-6.0833],[208,-6.005],[211,-5.8873],[216,-5.592],[222,-5.1163],[223,-5.0167],[224,-4.9369],[225,-4.837],[226,-4.757],[228,-4.5565],[229,-4.4761],[232,-4.1739],[233,-4.0931],[234,-3.9919],[235,-3.9109],[236,-3.8095],[242,-3.3209],[247,-3.014],[250,-2.8909],[254,-2.8088],[257,-2.8088],[258,-2.8293],[259,-2.8293],[260,-2.8499],[261,-2.8499],[268,-2.9935],[269,-3.0345],[270,-3.055],[271,-3.0959],[272,-3.1164],[273,-3.1573],[274,-3.1778],[275,-3.2187],[276,-3.2391],[278,-3.3209],[279,-3.3413],[281,-3.4229],[282,-3.4433],[284,-3.5248],[285,-3.5452],[286,-3.5859],[287,-3.6063],[288,-3.647],[289,-3.6673],[290,-3.708],[299,-3.8906],[300,-3.8906],[301,-3.9109],[319,-3.9109],[320,-0.315],[321,4.3741],[322,6.3872],[323,8.091],[324,9.3033],[325,9.7463],[360,9.7463],[361,9.4597],[362,8.6847],[363,7.488],[364,5.9201],[365,4.1179],[366,2.1847],[367,0.3043],[368,-1.3774],[369,-2.746],[370,-3.6416],[371,-3.9668],[439,-3.9669]]
       }
       CTRL_C_eyesAim.Rotation.X {
        type=MovieSceneScriptingFloatChannel
        default=0
        thinned_from=440
        keys [[0,-9.4436],[31,-9.4436],[32,-4.2398],[33,0.3299],[34,-3.943],[35,-12.017],[36,-16.4597],[44,-16.4597],[45,-21.162],[62,-21.162],[63,-4.064],[64,-5.2349],[65,-7.8715],[66,-11.637],[67,-13.7246],[68,-12.9715],[69,-10.893],[70,-9.4436],[142,-9.4436],[143,1.962],[144,14.7025],[145,14.5948],[146,13.8695],[147,12.9064],[148,11.9147],[149,10.8615],[150,9.7746],[151,6.619],[152,2.3273],[153,-0.1929],[154,-0.7721],[156,-1.5459],[158,-2.4189],[159,-2.9038],[160,-3.3418],[162,-3.7332],[163,-3.88],[164,-4.0769],[166,-4.3727],[167,-4.5702],[168,-4.7192],[170,-5.1138],[173,-5.8564],[174,-6.2005],[175,-6.4979],[176,-6.8937],[177,-7.241],[178,-7.6379],[179,-7.9863],[180,-8.3816],[181,-8.7309],[182,-9.0314],[183,-9.3816],[185,-9.9817],[190,-11.7332],[191,-12.1327],[192,-12.4334],[194,-13.1387],[195,-13.4365],[198,-14.1934],[199,-14.3923],[200,-14.5456],[202,-14.7446],[204,-14.7446],[208,-14.5456],[211,-14.2472],[216,-13.5015],[222,-12.3097],[223,-12.0616],[224,-11.8631],[225,-11.6151],[226,-11.4167],[228,-10.9209],[229,-10.7226],[232,-9.9794],[233,-9.7813],[234,-9.5337],[235,-9.3357],[236,-9.0881],[242,-7.9006],[247,-7.1589],[250,-6.8623],[254,-6.6646],[257,-6.6646],[258,-6.714],[259,-6.714],[260,-6.7634],[261,-6.7634],[268,-7.1095],[269,-7.2083],[270,-7.2578],[271,-7.3567],[272,-7.4061],[273,-7.505],[274,-7.5545],[275,-7.6534],[276,-7.7028],[278,-7.9006],[279,-7.9501],[281,-8.148],[282,-8.1974],[284,-8.3953],[285,-8.4448],[286,-8.5438],[287,-8.5932],[288,-8.6922],[289,-8.7417],[290,-8.8407],[299,-9.2862],[300,-9.2862],[301,-9.3357],[319,-9.3357],[320,-0.7402],[321,10.2293],[322,14.9843],[323,19.08],[324,22.0524],[325,23.1533],[360,23.1533],[361,22.4322],[362,20.5081],[363,17.5971],[364,13.8594],[365,9.6242],[366,5.1103],[367,0.7139],[368,-3.246],[369,-6.5026],[370,-8.6562],[371,-9.4436],[439,-9.4436]]
       }
       CTRL_C_eyesAim.Rotation.Y {
        type=MovieSceneScriptingFloatChannel
        default=0
        thinned_from=440
        keys [[0,0],[31,0],[32,-1.5447],[33,0.2768],[34,-3.3023],[35,-9.9097],[36,-13.3738],[44,-13.3738],[45,-16.8526],[62,-16.8526],[63,-3.4033],[64,-4.3779],[65,-6.5555],[66,-7.1665],[67,-6.5068],[68,-4.4249],[69,-1.2628],[70,0],[143,0],[144,-2.626],[145,-5.4315],[146,-5.4498],[147,-5.2732],[148,-5.0339],[149,-4.69],[150,-4.3026],[151,-2.9049],[152,-1.0083],[153,0.0827],[154,0.3294],[156,0.655],[158,1.0153],[159,1.2159],[160,1.3926],[161,1.4703],[162,1.5447],[163,1.6016],[164,1.6749],[166,1.7837],[167,1.8553],[168,1.9067],[169,1.9811],[170,2.0511],[173,2.3145],[174,2.4439],[175,2.5483],[176,2.6897],[177,2.8107],[178,2.9494],[179,3.0679],[180,3.2106],[181,3.3268],[182,3.4234],[183,3.5372],[184,3.6394],[185,3.7329],[186,3.8519],[190,4.3017],[191,4.4344],[193,4.6338],[195,4.8514],[199,5.1485],[202,5.2577],[204,5.2577],[208,5.1887],[211,5.085],[216,4.8252],[222,4.4079],[223,4.3207],[224,4.2509],[225,4.1636],[226,4.0937],[228,3.9187],[229,3.8486],[232,3.5854],[233,3.5151],[234,3.4271],[235,3.3567],[236,3.2687],[242,2.8452],[247,2.58],[250,2.4737],[254,2.4029],[257,2.4029],[258,2.4206],[259,2.4206],[260,2.4383],[261,2.4383],[268,2.5623],[269,2.5977],[270,2.6154],[271,2.6508],[272,2.6684],[273,2.7038],[274,2.7215],[275,2.7569],[276,2.7745],[278,2.8452],[279,2.8629],[281,2.9336],[282,2.9512],[284,3.0218],[285,3.0395],[286,3.0748],[287,3.0924],[288,3.1277],[289,3.1453],[290,3.1806],[299,3.3391],[300,3.3391],[301,3.3567],[319,3.3567],[320,0.2676],[321,-3.6739],[322,-5.3408],[323,-6.7407],[324,-7.7308],[325,-8.0913],[360,-8.0913],[361,-7.658],[362,-6.5508],[363,-5.0294],[364,-3.3904],[365,-1.9086],[366,-0.7724],[367,-0.0749],[368,0.2069],[369,0.1961],[370,0.0704],[371,0],[439,0]]
       }
       CTRL_C_eyesAim.Rotation.Z {
        type=MovieSceneScriptingFloatChannel
        default=0
        thinned_from=440
        keys [[0,0],[31,0],[33,-39.9992],[34,-39.8863],[36,-38.0571],[44,-38.0571],[45,-36.8298],[62,-36.8298],[63,-39.8792],[64,-39.7998],[65,-39.5485],[66,-31.2062],[67,-24.8912],[68,-18.5182],[69,-6.5332],[70,0],[143,0],[144,9.9044],[145,19.9775],[146,21.0392],[147,21.854],[148,22.5801],[149,23.0805],[150,23.5328],[151,23.5925],[153,23.1991],[155,23.0072],[156,22.9566],[158,22.7569],[159,22.7007],[160,22.5974],[161,22.5458],[162,22.4473],[163,22.3966],[165,22.2001],[166,22.1489],[168,21.9514],[169,21.8981],[173,21.4906],[174,21.4298],[179,20.8803],[180,20.8124],[183,20.4767],[184,20.412],[185,20.3006],[186,20.2317],[187,20.1159],[188,20.0456],[189,19.9287],[192,19.712],[193,19.5927],[194,19.5186],[195,19.4937],[198,19.2924],[199,19.2747],[200,19.2158],[319,19.6042],[320,19.8767],[321,19.5495],[322,19.1755],[323,18.7443],[324,18.3698],[325,18.2183],[360,18.2183],[361,17.8899],[362,16.9615],[363,15.4512],[364,13.4793],[365,11.1116],[366,8.5718],[367,5.9865],[368,3.6425],[369,1.7195],[370,0.4624],[371,0],[439,0]]
       }
       CTRL_C_eyesAim.Scale.X {type=MovieSceneScriptingFloatChannel, default=1, const [1,440]}
       CTRL_C_eyesAim.Scale.Y {type=MovieSceneScriptingFloatChannel, default=1, const [1,440]}
       CTRL_C_eyesAim.Scale.Z {type=MovieSceneScriptingFloatChannel, default=1, const [1,440]}
       CTRL_L_eyeAim.Location.X {type=MovieSceneScriptingFloatChannel, default=0, const [0,440]}
       CTRL_L_eyeAim.Location.Y {type=MovieSceneScriptingFloatChannel, default=0, const [0,440]}
       CTRL_L_eyeAim.Location.Z {type=MovieSceneScriptingFloatChannel, default=0, const [0,440]}
       CTRL_L_eyeAim.Rotation.X {type=MovieSceneScriptingFloatChannel, default=0, const [0,440]}
       CTRL_L_eyeAim.Rotation.Y {type=MovieSceneScriptingFloatChannel, default=0, const [0,440]}
       CTRL_L_eyeAim.Rotation.Z {type=MovieSceneScriptingFloatChannel, default=0, const [0,440]}
       CTRL_L_eyeAim.Scale.X {type=MovieSceneScriptingFloatChannel, default=1, const [1,440]}
       CTRL_L_eyeAim.Scale.Y {type=MovieSceneScriptingFloatChannel, default=1, const [1,440]}
       CTRL_L_eyeAim.Scale.Z {type=MovieSceneScriptingFloatChannel, default=1, const [1,440]}
       CTRL_R_eyeAim.Location.X {
        type=MovieSceneScriptingFloatChannel
        default=0
        thinned_from=440
        keys [[0,0],[31,0],[32,0.1481],[33,-0.0265],[34,0.3165],[35,0.9455],[36,1.2708],[44,1.2708],[45,1.5928],[62,1.5928],[63,0.3262],[64,0.4194],[65,0.6272],[66,0.6854],[67,0.6226],[68,0.4239],[69,0.1211],[70,0],[143,0],[144,0.2517],[145,0.52],[146,0.5218],[147,0.5049],[148,0.4821],[149,0.4492],[150,0.4122],[151,0.2784],[152,0.0967],[153,-0.0079],[154,-0.0316],[156,-0.0628],[158,-0.0973],[159,-0.1166],[160,-0.1335],[161,-0.141],[162,-0.1481],[163,-0.1536],[164,-0.1606],[166,-0.171],[167,-0.1779],[168,-0.1828],[169,-0.1899],[170,-0.1966],[173,-0.2219],[174,-0.2343],[175,-0.2443],[176,-0.2578],[177,-0.2694],[178,-0.2827],[179,-0.294],[180,-0.3077],[181,-0.3188],[182,-0.3281],[183,-0.3389],[184,-0.3487],[185,-0.3577],[186,-0.3691],[190,-0.4121],[191,-0.4248],[193,-0.4438],[195,-0.4646],[199,-0.493],[202,-0.5034],[204,-0.5034],[208,-0.4968],[211,-0.4869],[216,-0.4621],[222,-0.4222],[223,-0.4139],[224,-0.4072],[225,-0.3989],[226,-0.3922],[228,-0.3754],[229,-0.3687],[232,-0.3436],[233,-0.3368],[234,-0.3284],[235,-0.3217],[236,-0.3132],[242,-0.2727],[247,-0.2473],[250,-0.2371],[254,-0.2303],[257,-0.2303],[258,-0.232],[259,-0.232],[260,-0.2337],[261,-0.2337],[268,-0.2456],[269,-0.249],[270,-0.2507],[271,-0.2541],[272,-0.2558],[273,-0.2592],[274,-0.2609],[275,-0.2642],[276,-0.2659],[278,-0.2727],[279,-0.2744],[281,-0.2812],[282,-0.2829],[284,-0.2896],[285,-0.2913],[286,-0.2947],[287,-0.2964],[288,-0.2997],[289,-0.3014],[290,-0.3048],[299,-0.32],[300,-0.32],[301,-0.3217],[319,-0.3217],[320,-0.0257],[321,0.352],[322,0.5114],[323,0.6449],[324,0.739],[325,0.7733],[360,0.7733],[361,0.7321],[362,0.6268],[363,0.4816],[364,0.3249],[365,0.183],[366,0.0741],[367,0.0072],[368,-0.0198],[369,-0.0188],[370,-0.0067],[371,0],[439,0]]
       }
       CTRL_R_eyeAim.Location.Y {
        type=MovieSceneScriptingFloatChannel
        default=0
        thinned_from=440
        keys [[0,-0.9014],[31,-0.9014],[32,-0.406],[33,0.0316],[34,-0.3772],[35,-1.1268],[36,-1.5144],[44,-1.5144],[45,-1.8982],[62,-1.8982],[63,-0.3887],[64,-0.4998],[65,-0.7475],[66,-1.0995],[67,-1.2951],[68,-1.2295],[69,-1.038],[70,-0.9014],[142,-0.9014],[143,0.1881],[144,1.3929],[145,1.3781],[146,1.311],[147,1.2219],[148,1.1299],[149,1.0318],[150,0.9301],[151,0.6324],[152,0.2231],[153,-0.0185],[154,-0.074],[156,-0.1482],[158,-0.2318],[159,-0.2783],[160,-0.3202],[162,-0.3576],[163,-0.3716],[164,-0.3904],[166,-0.4187],[167,-0.4375],[168,-0.4517],[170,-0.4894],[173,-0.5601],[174,-0.5928],[175,-0.6211],[176,-0.6587],[177,-0.6916],[178,-0.7292],[179,-0.7622],[180,-0.7996],[181,-0.8325],[182,-0.8609],[183,-0.8938],[185,-0.9503],[190,-1.1141],[191,-1.1512],[192,-1.1792],[194,-1.2445],[195,-1.272],[198,-1.3418],[199,-1.36],[200,-1.3741],[202,-1.3924],[204,-1.3924],[208,-1.3741],[211,-1.3468],[216,-1.2781],[222,-1.1678],[223,-1.1448],[224,-1.1263],[225,-1.1032],[226,-1.0847],[228,-1.0384],[229,-1.0199],[232,-0.9502],[233,-0.9316],[234,-0.9083],[235,-0.8897],[236,-0.8664],[242,-0.7542],[247,-0.684],[250,-0.6558],[254,-0.637],[257,-0.637],[258,-0.6417],[259,-0.6417],[260,-0.6464],[261,-0.6464],[268,-0.6793],[269,-0.6887],[270,-0.6933],[271,-0.7027],[272,-0.7074],[273,-0.7168],[274,-0.7215],[275,-0.7308],[276,-0.7355],[278,-0.7542],[279,-0.7589],[281,-0.7776],[282,-0.7823],[284,-0.801],[285,-0.8057],[286,-0.815],[287,-0.8197],[288,-0.829],[289,-0.8337],[290,-0.843],[299,-0.885],[300,-0.885],[301,-0.8897],[319,-0.8897],[320,-0.071],[321,0.9736],[322,1.4143],[323,1.7835],[324,2.044],[325,2.1386],[360,2.1387],[361,2.0777],[362,1.9122],[363,1.6545],[364,1.3137],[365,0.918],[366,0.4893],[367,0.0685],[368,-0.3111],[369,-0.6222],[370,-0.8269],[371,-0.9014],[439,-0.9014]]
       }
       CTRL_R_eyeAim.Location.Z {
        type=MovieSceneScriptingFloatChannel
        default=0
        thinned_from=440
        keys [[0,0.0745],[31,0.0745],[32,0.017],[33,0.0001],[34,0.0221],[35,0.2006],[36,0.368],[44,0.368],[45,0.5905],[62,0.5905],[63,0.0235],[64,0.0389],[65,0.0873],[66,0.155],[67,0.1912],[68,0.1561],[69,0.1003],[70,0.0745],[142,0.0745],[143,0.0032],[144,0.1855],[145,0.2011],[146,0.1843],[147,0.1615],[149,0.1165],[150,0.095],[151,0.0436],[152,0.0054],[153,0],[155,0.0013],[156,0.0024],[157,0.0039],[158,0.0058],[160,0.011],[162,0.0137],[163,0.0147],[164,0.0163],[166,0.0187],[167,0.0203],[168,0.0217],[170,0.0254],[172,0.0304],[173,0.0331],[174,0.0371],[175,0.0407],[176,0.0457],[177,0.0504],[178,0.056],[179,0.0611],[180,0.0672],[181,0.0728],[182,0.0778],[183,0.0838],[185,0.0946],[186,0.1013],[187,0.1081],[189,0.1225],[190,0.13],[191,0.1388],[192,0.1456],[194,0.1622],[195,0.1695],[198,0.1886],[199,0.1939],[200,0.1979],[202,0.2033],[204,0.2033],[208,0.1979],[211,0.1899],[216,0.1708],[221,0.1468],[222,0.1422],[223,0.1366],[224,0.1321],[225,0.1267],[226,0.1225],[228,0.1121],[229,0.1081],[231,0.0984],[232,0.0937],[233,0.0901],[234,0.0856],[235,0.0821],[236,0.0778],[239,0.068],[242,0.0589],[247,0.0484],[250,0.0444],[254,0.0419],[257,0.0419],[258,0.0426],[259,0.0426],[260,0.0432],[261,0.0432],[268,0.0477],[269,0.049],[270,0.0497],[271,0.0511],[272,0.0517],[273,0.0531],[274,0.0538],[275,0.0552],[276,0.056],[278,0.0589],[279,0.0596],[281,0.0626],[282,0.0634],[284,0.0664],[285,0.0672],[286,0.0688],[287,0.0696],[288,0.0712],[289,0.072],[290,0.0736],[299,0.0812],[300,0.0812],[301,0.0821],[319,0.0821],[320,0.0005],[321,0.0984],[322,0.2098],[323,0.3377],[324,0.4482],[325,0.4928],[360,0.4928],[361,0.461],[362,0.3818],[363,0.2772],[364,0.1693],[365,0.0803],[366,0.0223],[367,0.0004],[368,0.0089],[369,0.0354],[370,0.0626],[371,0.0745],[439,0.0745]]
       }
       CTRL_R_eyeAim.Rotation.X {type=MovieSceneScriptingFloatChannel, default=0, thinned_from=440, keys [[0,0],[439,0]]}
       CTRL_R_eyeAim.Rotation.Y {type=MovieSceneScriptingFloatChannel, default=0, thinned_from=440, keys [[0,0],[439,0]]}
       CTRL_R_eyeAim.Rotation.Z {type=MovieSceneScriptingFloatChannel, default=0, thinned_from=440, keys [[0,0],[439,0]]}
       CTRL_R_eyeAim.Scale.X {type=MovieSceneScriptingFloatChannel, default=1, const [1,440]}
       CTRL_R_eyeAim.Scale.Y {type=MovieSceneScriptingFloatChannel, default=1, const [1,440]}
       CTRL_R_eyeAim.Scale.Z {type=MovieSceneScriptingFloatChannel, default=1, const [1,440]}
       CTRL_convergenceSwitch {type=MovieSceneScriptingFloatChannel, default=0, const [0,440]}
       CTRL_Master.Location.X {type=MovieSceneScriptingFloatChannel, default=0, const [0,440]}
       CTRL_Master.Location.Y {type=MovieSceneScriptingFloatChannel, default=0, const [0,440]}
       CTRL_Master.Location.Z {type=MovieSceneScriptingFloatChannel, default=0, const [0,440]}
       CTRL_Master.Rotation.X {type=MovieSceneScriptingFloatChannel, default=0, const [0,440]}
       CTRL_Master.Rotation.Y {type=MovieSceneScriptingFloatChannel, default=0, const [0,440]}
       CTRL_Master.Rotation.Z {type=MovieSceneScriptingFloatChannel, default=0, const [0,440]}
       CTRL_Master.Scale.X {type=MovieSceneScriptingFloatChannel, default=1, const [1,440]}
       CTRL_Master.Scale.Y {type=MovieSceneScriptingFloatChannel, default=1, const [1,440]}
       CTRL_Master.Scale.Z {type=MovieSceneScriptingFloatChannel, default=1, const [1,440]}
       CTRL_C_jaw.X {type=MovieSceneScriptingFloatChannel, default=0, const [0,440]}
       CTRL_C_jaw.Y {type=MovieSceneScriptingFloatChannel, default=0, const [0,440]}
       CTRL_C_jaw_openExtreme {type=MovieSceneScriptingFloatChannel, default=0, const [0,440]}
       CTRL_C_jaw_fwdBack {type=MovieSceneScriptingFloatChannel, default=0, thinned_from=440, keys [[0,0],[439,0]]}
       CTRL_R_jaw_ChinRaiseD {type=MovieSceneScriptingFloatChannel, default=0, const [0,440]}
       CTRL_R_jaw_ChinRaiseU {type=MovieSceneScriptingFloatChannel, default=0, const [0,440]}
       CTRL_L_jaw_ChinRaiseU {type=MovieSceneScriptingFloatChannel, default=0, const [0,440]}
       CTRL_L_jaw_ChinRaiseD {type=MovieSceneScriptingFloatChannel, default=0, const [0,440]}
       CTRL_R_mouth_lipsBlow {type=MovieSceneScriptingFloatChannel, default=0, const [0,440]}
       CTRL_L_mouth_lipsTogetherU {type=MovieSceneScriptingFloatChannel, default=0, const [0,440]}
       CTRL_L_mouth_lipsTogetherD {type=MovieSceneScriptingFloatChannel, default=0, const [0,440]}
       CTRL_R_mouth_lipsTogetherU {type=MovieSceneScriptingFloatChannel, default=0, const [0,440]}
       CTRL_L_mouth_lipsBlow {type=MovieSceneScriptingFloatChannel, default=0, const [0,440]}
       CTRL_R_mouth_lipsTogetherD {type=MovieSceneScriptingFloatChannel, default=0, const [0,440]}
       CTRL_C_tongue_roll.X {type=MovieSceneScriptingFloatChannel, default=0, const [0,440]}
       CTRL_C_tongue_roll.Y {type=MovieSceneScriptingFloatChannel, default=0, const [0,440]}
       CTRL_C_tongue.X {type=MovieSceneScriptingFloatChannel, default=0, const [0,440]}
       CTRL_C_tongue.Y {type=MovieSceneScriptingFloatChannel, default=0, const [0,440]}
       CTRL_C_tongue_narrowWide {type=MovieSceneScriptingFloatChannel, default=0, const [0,440]}
       CTRL_C_tongue_inOut {type=MovieSceneScriptingFloatChannel, default=0, const [0,440]}
       CTRL_R_nose.X {type=MovieSceneScriptingFloatChannel, default=0, const [0,440]}
       CTRL_R_nose.Y {type=MovieSceneScriptingFloatChannel, default=0, const [0,440]}
       CTRL_L_nose.X {type=MovieSceneScriptingFloatChannel, default=0, const [0,440]}
       CTRL_L_nose.Y {type=MovieSceneScriptingFloatChannel, default=0, const [0,440]}
       CTRL_L_eye.X {
        type=MovieSceneScriptingFloatChannel
        default=0
        thinned_from=440
        keys [[0,-0.2361],[24,-0.2361],[25,-0.1128],[72,-0.1128],[73,-0.1256],[74,-0.1562],[75,-0.1927],[76,-0.2233],[77,-0.2361],[142,-0.2361],[143,0],[319,0],[320,-0.0197],[321,0.2716],[322,0.3972],[323,0.5048],[324,0.5826],[325,0.6113],[360,0.6113],[361,0.591],[362,0.5371],[363,0.457],[364,0.3565],[365,0.2452],[366,0.1292],[367,0.0179],[368,-0.0813],[369,-0.1626],[370,-0.2164],[371,-0.2361],[439,-0.2361]]
       }
       CTRL_L_eye.Y {
        type=MovieSceneScriptingFloatChannel
        default=0
        thinned_from=440
        keys [[0,0],[24,0],[25,-0.25],[72,-0.25],[73,-0.2284],[74,-0.176],[75,-0.1111],[76,-0.0521],[77,0],[143,0],[144,0.2561],[145,0.3176],[146,0.314],[147,0.3039],[148,0.2883],[149,0.268],[150,0.2439],[151,0.2171],[152,0.1884],[154,0.1292],[155,0.1005],[156,0.0737],[157,0.0496],[158,0.0293],[159,0.0136],[160,0.0036],[161,0],[315,0],[316,0.1918],[317,0.497],[360,0.497],[361,0.4853],[362,0.4537],[363,0.4058],[364,0.3473],[365,0.2818],[366,0.2152],[367,0.1497],[368,0.0912],[369,0.0433],[370,0.0117],[371,0],[439,0]]
       }
       CTRL_L_eye_blink {
        type=MovieSceneScriptingFloatChannel
        default=0
        thinned_from=440
        keys [[0,0],[30,0],[35,0.0011],[40,0.0031],[42,0.0046],[44,0.0068],[46,0.0094],[48,0.0126],[49,0.0679],[50,0.0805],[64,0.0805],[65,0.2241],[67,0.8563],[68,1],[69,0.9519],[70,0.8231],[71,0.7632],[72,0.4154],[73,0.1827],[74,-0.0384],[75,-0.225],[76,-0.3538],[77,-0.4019],[135,-0.4019],[136,0.2545],[137,0.9491],[138,1],[140,1],[141,0.9706],[142,0.6869],[143,0.3053],[144,0.1389],[145,0.1284],[152,0.1085],[154,0.1034],[155,0.1011],[156,0.0991],[157,0.0973],[158,0.0959],[159,0.0948],[160,0.0942],[161,0.0939],[315,0.0939],[316,0.1294],[317,0.3875],[318,0.591],[319,0.683],[321,0.683],[322,0.5695],[323,0.3209],[324,0.0744],[325,-0.0392],[360,-0.0392],[361,0.0391],[362,0.225],[363,0.4481],[364,0.634],[365,0.7123],[367,0.7123],[368,0.6008],[370,0.1116],[371,0],[439,0]]
       }
       CTRL_R_eye.X {
        type=MovieSceneScriptingFloatChannel
        default=0
        thinned_from=440
        keys [[0,-0.2361],[24,-0.2361],[25,-0.1128],[72,-0.1128],[73,-0.1256],[74,-0.1562],[75,-0.1927],[76,-0.2233],[77,-0.2361],[139,-0.2361],[140,0],[319,0],[320,-0.0197],[321,0.2716],[322,0.3972],[323,0.5048],[324,0.5826],[325,0.6113],[360,0.6113],[361,0.591],[362,0.5371],[363,0.457],[364,0.3565],[365,0.2452],[366,0.1292],[367,0.0179],[368,-0.0813],[369,-0.1626],[370,-0.2164],[371,-0.2361],[439,-0.2361]]
       }
       CTRL_R_eye.Y {
        type=MovieSceneScriptingFloatChannel
        default=0
        thinned_from=440
        keys [[0,0],[24,0],[25,-0.25],[72,-0.25],[73,-0.2284],[74,-0.176],[75,-0.1111],[76,-0.0521],[77,0],[143,0],[144,0.2561],[145,0.3176],[146,0.314],[147,0.3039],[148,0.2883],[149,0.268],[150,0.2439],[151,0.2171],[152,0.1884],[154,0.1292],[155,0.1005],[156,0.0737],[157,0.0496],[158,0.0293],[159,0.0136],[160,0.0036],[161,0],[315,0],[316,0.1918],[317,0.497],[360,0.497],[361,0.4853],[362,0.4537],[363,0.4058],[364,0.3473],[365,0.2818],[366,0.2152],[367,0.1497],[368,0.0912],[369,0.0433],[370,0.0117],[371,0],[439,0]]
       }
       CTRL_R_eye_blink {
        type=MovieSceneScriptingFloatChannel
        default=0
        thinned_from=440
        keys [[0,0],[30,0],[35,0.0011],[40,0.0031],[42,0.0046],[44,0.0068],[46,0.0094],[48,0.0126],[49,0.0679],[50,0.0805],[64,0.0805],[65,0.2241],[67,0.8563],[68,1],[69,0.9552],[70,0.8352],[71,0.7794],[72,0.4554],[73,0.2386],[74,0.0326],[75,-0.1412],[76,-0.2612],[77,-0.306],[135,-0.306],[136,0.3025],[137,0.9491],[138,1],[140,1],[141,0.9706],[142,0.6869],[143,0.317],[144,0.0861],[145,0.0854],[147,0.0857],[149,0.0867],[157,0.0925],[159,0.0935],[161,0.0938],[315,0.0938],[316,0.1294],[317,0.3875],[318,0.591],[319,0.683],[321,0.683],[322,0.5695],[323,0.3209],[324,0.0744],[325,-0.0392],[360,-0.0392],[361,0.0391],[362,0.225],[363,0.4481],[364,0.634],[365,0.7123],[367,0.7123],[368,0.6008],[370,0.1116],[371,0],[439,0]]
       }
       CTRL_C_eye.X {
        type=MovieSceneScriptingFloatChannel
        default=0
        thinned_from=440
        keys [[0,-0.3385],[7,-0.3385],[8,-0.0716],[10,-0.0716],[11,-0.0667],[23,-0.0667],[24,-0.3566],[25,-0.6987],[26,-0.6486],[27,-0.6123],[28,-0.5972],[63,-0.5972],[64,-0.8876],[70,-0.8876],[71,-0.8913],[72,-0.9119],[73,-0.8725],[74,-0.7785],[75,-0.6664],[76,-0.5725],[77,-0.5331],[90,-0.5331],[91,-0.513],[92,-0.4697],[93,-0.4288],[94,-0.4002],[95,-0.3872],[119,-0.3872],[120,-0.2269],[135,-0.2269],[137,0],[138,0.008],[139,0.0301],[140,0.063],[141,0.1037],[142,0.149],[143,0.1959],[144,0.2413],[145,0.2916],[146,0.2889],[147,0.2814],[148,0.2696],[149,0.2539],[150,0.2348],[151,0.213],[152,0.1888],[153,0.1628],[154,0.1356],[155,0.1076],[156,0.0793],[157,0.0513],[158,0.0241],[159,-0.0018],[160,-0.0259],[161,-0.0477],[162,-0.0688],[163,-0.0913],[164,-0.1148],[165,-0.1394],[166,-0.1646],[167,-0.1905],[172,-0.3229],[174,-0.374],[175,-0.3984],[176,-0.4219],[177,-0.4443],[178,-0.4653],[179,-0.4849],[180,-0.5028],[181,-0.5189],[182,-0.533],[183,-0.5449],[184,-0.5545],[185,-0.5615],[186,-0.5659],[187,-0.5674],[188,-0.5663],[189,-0.5632],[190,-0.5583],[191,-0.5516],[192,-0.5433],[193,-0.5336],[194,-0.5226],[196,-0.4975],[198,-0.469],[205,-0.3606],[207,-0.3322],[208,-0.3191],[209,-0.307],[210,-0.296],[211,-0.2863],[212,-0.2781],[213,-0.2714],[214,-0.2664],[215,-0.2633],[216,-0.2622],[219,-0.2634],[222,-0.2665],[239,-0.2985],[243,-0.3038],[246,-0.3056],[315,-0.3057],[316,-0.3394],[317,-0.3929],[318,-0.3844],[319,-0.3686],[320,-0.3601],[321,-0.3567],[323,-0.3419],[324,-0.3385],[362,-0.3385],[363,-0.1947],[364,-0.1947],[365,-0.2676],[366,-0.3385],[368,-0.3385],[369,-0.3402],[370,-0.3031],[371,-0.2342],[372,-0.1971],[408,-0.1971],[409,-0.3385],[439,-0.3385]]
       }
       CTRL_C_eye.Y {
        type=MovieSceneScriptingFloatChannel
        default=0
        thinned_from=440
        keys [[0,-0.2686],[10,-0.2686],[11,-0.1302],[23,-0.1302],[24,-0.233],[25,-0.4741],[63,-0.4741],[64,-0.4379],[70,-0.4379],[71,-0.4342],[72,-0.3995],[73,-0.3535],[74,-0.3113],[75,-0.2765],[76,-0.2529],[77,-0.2442],[90,-0.2442],[91,-0.229],[92,-0.1953],[93,-0.0945],[94,-0.0609],[95,-0.0456],[119,-0.0456],[120,-0.1966],[135,-0.1966],[137,-0.2686],[144,-0.2686],[145,-0.0814],[146,-0.0437],[147,-0.0063],[148,0.0305],[149,0.0664],[150,0.1011],[151,0.1343],[152,0.1658],[153,0.1953],[154,0.2223],[155,0.2467],[156,0.2682],[157,0.2864],[158,0.3011],[159,0.312],[160,0.3187],[161,0.321],[162,0.3195],[163,0.3153],[164,0.3085],[165,0.2994],[166,0.2881],[167,0.2749],[168,0.2599],[169,0.2435],[170,0.2257],[171,0.2069],[172,0.1871],[173,0.1667],[177,0.0826],[178,0.062],[179,0.042],[180,0.0227],[181,0.0045],[182,-0.0125],[183,-0.0281],[184,-0.042],[185,-0.0541],[186,-0.0641],[187,-0.0719],[192,-0.1044],[195,-0.123],[197,-0.1349],[199,-0.1461],[201,-0.1566],[203,-0.1662],[205,-0.1748],[207,-0.1823],[209,-0.1885],[211,-0.1934],[213,-0.1968],[214,-0.1978],[215,-0.1985],[216,-0.1987],[217,-0.1983],[218,-0.1969],[219,-0.1948],[220,-0.1919],[221,-0.1883],[222,-0.1841],[223,-0.1794],[224,-0.1741],[225,-0.1683],[226,-0.1622],[228,-0.149],[234,-0.1063],[235,-0.0994],[236,-0.0926],[237,-0.0862],[238,-0.0801],[239,-0.0743],[240,-0.069],[241,-0.0643],[242,-0.06],[243,-0.0565],[244,-0.0536],[245,-0.0514],[246,-0.0501],[247,-0.0497],[315,-0.0497],[316,-0.3961],[317,-0.9472],[318,-0.8431],[319,-0.6499],[320,-0.5458],[321,-0.5025],[323,-0.3119],[324,-0.2686],[363,-0.2686],[364,-0.3753],[365,-0.3324],[366,-0.3753],[367,-0.2686],[368,-0.2686],[369,-0.0109],[370,-0.041],[371,-0.097],[372,-0.1271],[408,-0.1271],[409,-0.2686],[439,-0.2686]]
       }
       CTRL_C_eyesConvergence {type=MovieSceneScriptingFloatChannel, default=0, const [0,440]}
       CTRL_L_eye_cheekRaise {
        type=MovieSceneScriptingFloatChannel
        default=0
        thinned_from=292
        keys [[0,0],[32,0],[34,0.3669],[43,0.3694],[49,0.3694],[62,0.3669],[63,0],[73,0],[77,0.0513],[136,0.0513],[137,0],[439,0]]
       }
       CTRL_L_eye_squintInner {
        type=MovieSceneScriptingFloatChannel
        default=0
        thinned_from=440
        keys [[0,0.3933],[23,0.3933],[24,0.3437],[25,0.2042],[26,0.102],[27,0.102],[29,0],[67,0],[68,0.007],[69,0.0225],[70,0.0697],[71,0.097],[72,0.274],[73,0.3874],[74,0.4916],[75,0.5774],[76,0.6356],[77,0.6571],[135,0.6571],[136,0.5544],[138,0.1027],[139,0],[407,0],[408,0.0615],[410,0.3318],[411,0.3933],[439,0.3933]]
       }
       CTRL_L_nose_wrinkleUpper {type=MovieSceneScriptingFloatChannel, default=0, const [1,440]}
       CTRL_R_nose_wrinkleUpper {type=MovieSceneScriptingFloatChannel, default=0, const [1,440]}
       CTRL_R_eye_squintInner {
        type=MovieSceneScriptingFloatChannel
        default=0
        thinned_from=440
        keys [[0,0.3933],[23,0.3933],[24,0.3437],[25,0.2042],[26,0.102],[27,0.102],[29,0],[67,0],[68,0.007],[69,0.0225],[70,0.0697],[71,0.097],[72,0.274],[73,0.3874],[74,0.4916],[75,0.5774],[76,0.6356],[77,0.6571],[135,0.6571],[136,0.5544],[138,0.1027],[139,0],[407,0],[408,0.0615],[410,0.3318],[411,0.3933],[439,0.3933]]
       }
       CTRL_R_eye_cheekRaise {
        type=MovieSceneScriptingFloatChannel
        default=0
        thinned_from=440
        keys [[0,0],[26,0],[28,0.37],[62,0.37],[63,-0.0057],[64,-0.0053],[69,-0.0015],[70,-0.0011],[71,0],[72,0.005],[73,0.005],[74,0.0129],[76,0.0477],[77,0.0557],[131,0.0549],[132,0.0547],[133,0.0543],[136,0.052],[137,0.0013],[138,0.0006],[139,0.0002],[439,0]]
       }
       CTRL_C_eyelidsReduce {type=MovieSceneScriptingFloatChannel, default=0, const [0,440]}
       CTRL_R_brow_lateral {
        type=MovieSceneScriptingFloatChannel
        default=0
        thinned_from=440
        keys [[0,0.4751],[25,0.4751],[26,0.64],[27,0.6806],[28,0.7118],[29,0.7243],[49,0.7243],[50,0.7018],[63,0.7018],[64,0.6474],[65,0.5158],[66,0.3546],[67,0.2112],[68,0.133],[69,0.1053],[70,0.0894],[71,0.0833],[72,0.0521],[73,0.0367],[74,0.0225],[75,0.0108],[76,0.0029],[77,0],[407,0],[408,0.0742],[410,0.4009],[411,0.4751],[439,0.4751]]
       }
       CTRL_L_brow_lateral {
        type=MovieSceneScriptingFloatChannel
        default=0
        thinned_from=440
        keys [[0,0.1099],[25,0.1099],[26,0.5967],[27,0.7539],[28,0.8749],[29,0.9232],[49,0.9232],[63,0.9248],[64,0.8472],[65,0.6603],[66,0.4331],[67,0.2343],[68,0.133],[69,0.1053],[70,0.0894],[71,0.0833],[72,0.0521],[73,0.0367],[74,0.0225],[75,0.0108],[76,0.0029],[77,0],[407,0],[408,0.0172],[410,0.0927],[411,0.1099],[439,0.1099]]
       }
       CTRL_R_brow_down {
        type=MovieSceneScriptingFloatChannel
        default=0
        thinned_from=440
        keys [[0,0.5112],[25,0.5112],[26,0.5271],[27,0.5731],[28,0.6337],[29,0.6639],[49,0.6639],[51,0.569],[62,0.589],[63,0.5875],[65,0.5788],[66,0.5763],[67,0.5777],[68,0.5824],[69,0.5897],[70,0.5987],[71,0.6016],[72,0.6175],[73,0.6256],[74,0.6315],[75,0.634],[135,0.6341],[137,0.5339],[407,0.5339],[408,0.5303],[410,0.5147],[411,0.5112],[439,0.5112]]
       }
       CTRL_L_brow_down {
        type=MovieSceneScriptingFloatChannel
        default=0
        thinned_from=440
        keys [[0,0.5112],[25,0.5112],[26,0.5314],[27,0.5899],[28,0.6668],[29,0.7051],[49,0.7051],[51,0.605],[64,0.6251],[65,0.6308],[67,0.65],[68,0.6541],[69,0.6511],[70,0.6431],[71,0.6394],[72,0.6178],[74,0.5896],[75,0.578],[76,0.57],[77,0.567],[135,0.567],[137,0.5339],[407,0.5339],[408,0.5303],[410,0.5147],[411,0.5112],[439,0.5112]]
       }
       CTRL_L_brow_raiseOut {type=MovieSceneScriptingFloatChannel, default=0, const [0,440]}
       CTRL_L_brow_raiseIn {
        type=MovieSceneScriptingFloatChannel
        default=0
        thinned_from=440
        keys [[0,0],[69,0],[70,0.0103],[71,0.0183],[72,0.0697],[73,0.1033],[74,0.1291],[75,0.1394],[76,0.1387],[77,0.1369],[78,0.1348],[79,0.1331],[80,0.1323],[135,0.1333],[137,0.0579],[166,0.0576],[186,0.0549],[315,0.0549],[316,0.0393],[318,0.0393],[319,-0.0207],[320,-0.1321],[321,-0.192],[322,-0.1246],[323,-0.0414],[324,-0.0123],[325,0.0009],[330,0.0009],[331,0.0003],[439,0]]
       }
       CTRL_R_brow_raiseIn {
        type=MovieSceneScriptingFloatChannel
        default=0
        thinned_from=440
        keys [[0,0],[69,0],[70,0.0103],[71,0.0182],[72,0.0695],[73,0.1029],[74,0.1287],[75,0.139],[76,0.1382],[77,0.1365],[78,0.1344],[79,0.1326],[80,0.1319],[135,0.1328],[137,0.0575],[166,0.0572],[186,0.0545],[315,0.0545],[316,0.0541],[318,0.0541],[319,-0.0155],[320,-0.1449],[321,-0.2145],[322,-0.1395],[323,-0.0468],[324,-0.0145],[325,0.0003],[439,0]]
       }
       CTRL_R_brow_raiseOut {
        type=MovieSceneScriptingFloatChannel
        default=0
        thinned_from=440
        keys [[0,0.2644],[66,0.2644],[67,0.2874],[68,0.3381],[70,0.3619],[71,0.3652],[72,0.3833],[73,0.3923],[74,0.3999],[75,0.4057],[76,0.4093],[77,0.4106],[315,0.406],[316,0.4043],[318,0.4043],[319,0.2851],[320,0.0637],[321,-0.0556],[322,0.0563],[323,0.1945],[324,0.2428],[325,0.2648],[439,0.2644]]
       }
       CTRL_lookAtSwitch {type=MovieSceneScriptingFloatChannel, default=0, const [0,440]}
       CTRL_C_teethD.X {type=MovieSceneScriptingFloatChannel, default=0, const [0,440]}
       CTRL_C_teethD.Y {type=MovieSceneScriptingFloatChannel, default=0, const [0,440]}
       CTRL_C_teethU.X {type=MovieSceneScriptingFloatChannel, default=0, const [0,440]}
       CTRL_C_teethU.Y {type=MovieSceneScriptingFloatChannel, default=0, const [0,440]}
       CTRL_C_teeth_fwdBackU {type=MovieSceneScriptingFloatChannel, default=0, const [0,440]}
       CTRL_L_mouth_corner.X {type=MovieSceneScriptingFloatChannel, default=0, const [0,440]}
       CTRL_L_mouth_corner.Y {type=MovieSceneScriptingFloatChannel, default=0, const [0,440]}
       CTRL_R_mouth_corner.X {type=MovieSceneScriptingFloatChannel, default=0, const [0,440]}
       CTRL_R_mouth_corner.Y {type=MovieSceneScriptingFloatChannel, default=0, const [0,440]}
       CTRL_C_teeth_fwdBackD {type=MovieSceneScriptingFloatChannel, default=0, const [0,440]}
       CTRL_R_mouth_cornerSharpnessD {type=MovieSceneScriptingFloatChannel, default=0, const [0,440]}
       CTRL_R_mouth_cornerSharpnessU {type=MovieSceneScriptingFloatChannel, default=0, const [0,440]}
       CTRL_L_eye_eyelidU.X {
        type=MovieSceneScriptingFloatChannel
        default=0
        thinned_from=440
        keys [[0,0],[74,0],[75,-0.0007],[76,-0.0019],[77,-0.0026],[135,-0.0026],[137,0],[439,0]]
       }
       CTRL_L_eye_eyelidU.Y {
        type=MovieSceneScriptingFloatChannel
        default=0
        thinned_from=440
        keys [[0,0],[71,0],[72,0.0206],[73,0.0697],[74,0.1282],[75,0.0304],[76,-0.2218],[77,-0.3688],[131,-0.3688],[132,-0.3755],[133,-0.3937],[134,-0.4201],[135,-0.4513],[136,-0.2008],[137,0.0513],[138,0.025],[139,0.0068],[140,0],[317,0],[318,-0.0429],[319,-0.3762],[320,-0.3796],[321,-0.3796],[322,-0.3543],[323,-0.0599],[324,0],[439,0]]
       }
       CTRL_L_mouth_cornerSharpnessD {type=MovieSceneScriptingFloatChannel, default=0, const [0,440]}
       CTRL_L_eye_eyelidD.X {
        type=MovieSceneScriptingFloatChannel
        default=0
        thinned_from=440
        keys [[0,0],[74,0],[75,0.1117],[76,0.3191],[77,0.4308],[135,0.4308],[137,0],[315,0],[316,0.3963],[317,0],[439,0]]
       }
       CTRL_L_eye_eyelidD.Y {
        type=MovieSceneScriptingFloatChannel
        default=0
        thinned_from=440
        keys [[0,-0.5224],[23,-0.5224],[25,0],[71,0],[72,-0.0272],[73,-0.0919],[74,-0.1692],[75,-0.2537],[76,-0.3174],[77,-0.3371],[131,-0.3371],[132,-0.3281],[133,-0.3041],[134,-0.2694],[135,-0.2282],[136,-0.1469],[137,-0.0677],[138,-0.033],[139,-0.009],[140,0],[141,0.1219],[142,0.0809],[143,-0.0188],[144,-0.143],[145,-0.2385],[146,-0.2779],[147,-0.28],[148,-0.2833],[149,-0.2877],[150,-0.293],[151,-0.2991],[153,-0.3131],[158,-0.3519],[160,-0.3661],[163,-0.3848],[175,-0.4734],[178,-0.4928],[180,-0.5036],[182,-0.5124],[184,-0.5186],[186,-0.5219],[314,-0.5224],[315,-0.4408],[317,-0.0816],[318,0],[322,0],[323,-0.1354],[324,-0.3869],[325,-0.5224],[362,-0.5224],[364,0],[368,0],[369,-0.0816],[371,-0.4408],[372,-0.5224],[439,-0.5224]]
       }
       CTRL_L_mouth_cornerSharpnessU {type=MovieSceneScriptingFloatChannel, default=0, const [0,440]}
       CTRL_R_mouth_lipsTowardsTeethD {type=MovieSceneScriptingFloatChannel, default=0, const [0,440]}
       CTRL_L_mouth_lipsTowardsTeethD {type=MovieSceneScriptingFloatChannel, default=0, const [0,440]}
       CTRL_R_eye_eyelidU.X {
        type=MovieSceneScriptingFloatChannel
        default=0
        thinned_from=440
        keys [[0,-0.4241],[23,-0.4241],[25,0],[74,0],[75,-0.1142],[76,-0.3262],[77,-0.6846],[135,-0.6846],[136,-0.2202],[137,0],[140,0],[141,-0.0314],[142,-0.11],[144,-0.3142],[145,-0.3927],[146,-0.4241],[161,-0.4241],[162,-0.4256],[163,-0.43],[164,-0.4371],[165,-0.4465],[166,-0.4582],[167,-0.4717],[168,-0.4869],[169,-0.5036],[171,-0.5403],[177,-0.66],[179,-0.6967],[180,-0.7134],[181,-0.7286],[182,-0.7421],[183,-0.7537],[184,-0.7632],[185,-0.7702],[186,-0.7746],[314,-0.7762],[315,-0.7099],[316,-0.5947],[317,-0.4928],[318,-0.4265],[319,0],[322,0],[323,-0.11],[324,-0.3142],[325,-0.4241],[362,-0.4241],[364,0],[368,0],[369,-0.0663],[371,-0.3578],[372,-0.4241],[439,-0.4241]]
       }
       CTRL_R_eye_eyelidU.Y {
        type=MovieSceneScriptingFloatChannel
        default=0
        thinned_from=440
        keys [[0,0.1586],[71,0.1586],[72,0.1792],[73,0.2283],[74,0.2868],[75,0.1174],[76,0.0267],[77,-0.0332],[131,-0.0332],[132,-0.04],[133,-0.0582],[134,-0.0845],[135,-0.1157],[136,-0.033],[137,0.2099],[138,0.1836],[139,0.1654],[140,0.1586],[315,0.1586],[316,0.1839],[317,0.2243],[318,0.2243],[319,-0.3762],[320,-0.3796],[321,-0.3796],[322,-0.3543],[323,-0.0599],[324,0.1586],[439,0.1586]]
       }
       CTRL_L_mouth_lipsTowardsTeethU {type=MovieSceneScriptingFloatChannel, default=0, const [0,440]}
       CTRL_R_eye_eyelidD.X {
        type=MovieSceneScriptingFloatChannel
        default=0
        thinned_from=440
        keys [[0,-0.3244],[23,-0.3244],[25,0],[74,0],[75,0.0521],[76,0.1489],[77,0.201],[79,0.201],[80,-0.3244],[81,0.201],[135,0.201],[137,0],[140,0],[141,-0.0145],[142,-0.0506],[144,-0.1445],[145,-0.1807],[146,-0.1951],[161,-0.1951],[162,-0.1947],[163,-0.1934],[164,-0.1914],[165,-0.1887],[166,-0.1854],[167,-0.1816],[168,-0.1773],[170,-0.1674],[172,-0.1565],[176,-0.1336],[178,-0.1227],[179,-0.1176],[180,-0.1128],[181,-0.1085],[182,-0.1047],[183,-0.1014],[184,-0.0987],[185,-0.0967],[186,-0.0954],[187,-0.095],[314,-0.095],[315,-0.0645],[316,-0.0139],[317,0.1863],[318,0.2168],[319,0],[322,0],[323,0.0101],[324,0.0289],[325,0.039],[362,0.039],[364,0],[368,0],[369,-0.0507],[371,-0.2737],[372,-0.3244],[439,-0.3244]]
       }
       CTRL_R_eye_eyelidD.Y {
        type=MovieSceneScriptingFloatChannel
        default=0
        thinned_from=440
        keys [[0,-0.5224],[23,-0.5224],[25,0],[71,0],[72,-0.0272],[73,-0.0919],[74,-0.1692],[75,-0.3156],[76,-0.4943],[77,-0.5759],[79,-0.5759],[80,-0.5224],[81,-0.5759],[131,-0.5759],[132,-0.567],[133,-0.543],[134,-0.5082],[135,-0.467],[136,-0.2663],[137,-0.0677],[138,-0.033],[139,-0.009],[140,0],[141,0.1219],[142,0.0809],[143,-0.0188],[144,-0.143],[145,-0.2385],[146,-0.2779],[147,-0.28],[148,-0.2833],[149,-0.2877],[150,-0.293],[151,-0.2991],[153,-0.3131],[158,-0.3519],[160,-0.3661],[163,-0.3848],[175,-0.4734],[178,-0.4928],[180,-0.5036],[182,-0.5124],[184,-0.5186],[186,-0.5219],[314,-0.5224],[315,-0.4408],[317,-0.0816],[318,0],[322,0],[323,-0.1049],[324,-0.2998],[325,-0.4048],[362,-0.4048],[364,0],[368,0],[369,-0.0816],[371,-0.4408],[372,-0.5224],[439,-0.5224]]
       }
       CTRL_R_mouth_lipsTowardsTeethU {type=MovieSceneScriptingFloatChannel, default=0, const [0,440]}
       CTRL_faceGUIfollowHead {type=MovieSceneScriptingFloatChannel, default=1, const [1,440]}
       CTRL_eyesAimFollowHead {type=MovieSceneScriptingFloatChannel, default=0, const [0,440]}
       CTRL_C_nose.X {type=MovieSceneScriptingFloatChannel, default=0}
       CTRL_C_nose.Y {type=MovieSceneScriptingFloatChannel, default=0, const [-0.7293,1]}
       CTRL_C_teeth_rotateU {type=MovieSceneScriptingFloatChannel, default=0, const [0,440]}
       CTRL_C_teeth_rotateD {type=MovieSceneScriptingFloatChannel, default=0, const [0,440]}
       CTRL_L_brow_inner {
        type=MovieSceneScriptingFloatChannel
        default=-0.3112
        thinned_from=440
        keys [[0,-0.3113],[23,-0.3113],[24,-0.4611],[25,-0.6068],[26,-0.6894],[63,-0.6894],[64,-0.7812],[69,-0.7812],[70,-0.7769],[71,-0.7696],[72,-0.7174],[73,-0.6867],[74,-0.6845],[75,-0.698],[76,-0.7139],[77,-0.7192],[78,-0.7149],[135,-0.7149],[136,-0.7359],[137,-0.7884],[138,-0.8565],[139,-1.0031],[142,-1.0011],[143,-0.8848],[144,-0.8619],[145,-0.8508],[181,-0.7686],[315,-0.7619],[316,-0.7663],[317,-0.7911],[320,-0.8535],[321,-0.8649],[322,-0.8433],[323,-0.8056],[324,-0.7663],[325,-0.7368],[327,-0.7206],[328,-0.7138],[329,-0.7087],[330,-0.7014],[361,-0.5998],[362,-0.618],[363,-0.6677],[364,-0.7075],[365,-0.7288],[366,-0.5476],[367,-0.4257],[368,-0.3945],[369,-0.3699],[370,-0.3587],[383,-0.3323],[393,-0.3187],[402,-0.3122],[439,-0.3112]]
       }
       CTRL_R_brow_inner {
        type=MovieSceneScriptingFloatChannel
        default=-0.3112
        thinned_from=440
        keys [[0,-0.3113],[23,-0.3113],[24,-0.4611],[25,-0.6068],[26,-0.6894],[63,-0.6894],[64,-0.7812],[70,-0.7812],[71,-0.7658],[72,-0.6452],[73,-0.5679],[74,-0.5572],[75,-0.5846],[76,-0.6219],[77,-0.6406],[135,-0.6406],[136,-0.6764],[137,-0.7603],[138,-0.8565],[139,-1.0031],[142,-1.0011],[143,-0.8848],[144,-0.8619],[145,-0.8508],[181,-0.7686],[315,-0.7619],[316,-0.7663],[317,-0.7911],[320,-0.8535],[321,-0.8649],[322,-0.8433],[323,-0.8056],[324,-0.7663],[325,-0.7368],[327,-0.7206],[328,-0.7138],[329,-0.7087],[330,-0.7014],[361,-0.5998],[362,-0.618],[363,-0.6677],[364,-0.7075],[365,-0.7288],[366,-0.5476],[367,-0.4257],[368,-0.3945],[369,-0.3699],[370,-0.3587],[383,-0.3323],[393,-0.3187],[402,-0.3122],[439,-0.3112]]
       }
       CTRL_L_brow_outer {
        type=MovieSceneScriptingFloatChannel
        default=0
        thinned_from=440
        keys [[0,0],[23,0],[24,-0.1399],[25,-0.3037],[26,-0.3781],[70,-0.3781],[71,-0.3709],[72,-0.3034],[73,-0.2454],[74,-0.1837],[75,-0.1607],[76,-0.1791],[77,-0.1862],[135,-0.1862],[136,-0.2072],[137,-0.2597],[138,-0.2429],[139,-0.2741],[140,-0.1891],[142,-0.1891],[143,-0.0527],[144,-0.0165],[145,0],[361,0],[362,-0.0443],[364,-0.2393],[365,-0.2836],[366,-0.1791],[367,-0.0597],[368,-0.031],[369,-0.0088],[370,0],[439,0]]
       }
       CTRL_R_brow_outer {
        type=MovieSceneScriptingFloatChannel
        default=0
        thinned_from=440
        keys [[0,0],[23,0],[24,-0.1399],[25,-0.3037],[26,-0.3781],[70,-0.3781],[71,-0.3668],[72,-0.2603],[73,-0.1442],[74,-0.0178],[75,0.0879],[76,0.1536],[77,0.1761],[135,0.1761],[136,0.1426],[137,0.061],[138,-0.0297],[139,-0.1994],[140,-0.1891],[142,-0.1891],[143,-0.0527],[144,-0.0165],[145,0],[361,0],[362,-0.0443],[364,-0.2393],[365,-0.2836],[366,-0.1791],[367,-0.0597],[368,-0.031],[369,-0.0088],[370,0],[439,0]]
       }
       CTRL_C_mouth.X {type=MovieSceneScriptingFloatChannel, default=0, const [0,440]}
       CTRL_C_mouth.Y {type=MovieSceneScriptingFloatChannel, default=0, const [0,440]}
       CTRL_L_mouth_upperLipRaise {type=MovieSceneScriptingFloatChannel, default=0, const [0,440]}
       CTRL_L_mouth_cornerPull {type=MovieSceneScriptingFloatChannel, default=0, const [0,440]}
       CTRL_L_mouth_dimple {type=MovieSceneScriptingFloatChannel, default=0, const [0,440]}
       CTRL_L_mouth_cornerDepress {type=MovieSceneScriptingFloatChannel, default=0, const [0,440]}
       CTRL_L_mouth_stretch {type=MovieSceneScriptingFloatChannel, default=0, const [0,440]}
       CTRL_L_mouth_lowerLipDepress {type=MovieSceneScriptingFloatChannel, default=0, const [0,440]}
       CTRL_R_mouth_lowerLipDepress {type=MovieSceneScriptingFloatChannel, default=0, const [0,440]}
       CTRL_R_mouth_stretch {type=MovieSceneScriptingFloatChannel, default=0, const [0,440]}
       CTRL_L_mouth_towardsD {type=MovieSceneScriptingFloatChannel, default=0, const [0,440]}
       CTRL_R_mouth_funnelU {type=MovieSceneScriptingFloatChannel, default=0, const [0,440]}
       CTRL_L_mouth_towardsU {type=MovieSceneScriptingFloatChannel, default=0, const [0,440]}
       CTRL_R_mouth_lipBiteD {type=MovieSceneScriptingFloatChannel, default=0, const [0,440]}
       CTRL_R_mouth_funnelD {type=MovieSceneScriptingFloatChannel, default=0, const [0,440]}
       CTRL_L_mouth_pressU {type=MovieSceneScriptingFloatChannel, default=0, const [0,440]}
       CTRL_L_mouth_funnelD {type=MovieSceneScriptingFloatChannel, default=0, const [0,440]}
       CTRL_L_mouth_suckBlow {type=MovieSceneScriptingFloatChannel, default=0, const [0,440]}
       CTRL_L_mouth_pressD {type=MovieSceneScriptingFloatChannel, default=0, const [0,440]}
       CTRL_L_mouth_funnelU {type=MovieSceneScriptingFloatChannel, default=0, const [0,440]}
       CTRL_R_mouth_towardsU {type=MovieSceneScriptingFloatChannel, default=0, const [0,440]}
       CTRL_R_mouth_purseD {type=MovieSceneScriptingFloatChannel, default=0, const [0,440]}
       CTRL_L_mouth_lipBiteU {type=MovieSceneScriptingFloatChannel, default=0, const [0,440]}
       CTRL_R_mouth_pressD {type=MovieSceneScriptingFloatChannel, default=0, const [0,440]}
       CTRL_R_mouth_pressU {type=MovieSceneScriptingFloatChannel, default=0, const [0,440]}
       CTRL_L_mouth_purseD {type=MovieSceneScriptingFloatChannel, default=0, const [0,440]}
       CTRL_R_mouth_lipBiteU {type=MovieSceneScriptingFloatChannel, default=0, const [0,440]}
       CTRL_R_mouth_suckBlow {type=MovieSceneScriptingFloatChannel, default=0, const [0,440]}
       CTRL_L_mouth_purseU {type=MovieSceneScriptingFloatChannel, default=0, const [0,440]}
       CTRL_R_mouth_towardsD {type=MovieSceneScriptingFloatChannel, default=0, const [0,440]}
       CTRL_L_mouth_lipBiteD {type=MovieSceneScriptingFloatChannel, default=0, const [0,440]}
       CTRL_R_mouth_purseU {type=MovieSceneScriptingFloatChannel, default=0, const [0,440]}
       CTRL_R_mouth_cornerDepress {type=MovieSceneScriptingFloatChannel, default=0, const [0,440]}
       CTRL_R_mouth_upperLipRaise {type=MovieSceneScriptingFloatChannel, default=0, const [0,440]}
       CTRL_R_mouth_dimple {type=MovieSceneScriptingFloatChannel, default=0, const [0,440]}
       CTRL_R_mouth_cornerPull {type=MovieSceneScriptingFloatChannel, default=0, const [0,440]}
       Weight {type=MovieSceneScriptingFloatChannel, default=1}
      ]
     }]
   }
   {
    type=MovieSceneControlRigParameterTrack
    display_name=MR_Biped_v02
    sections [{type=MovieSceneControlRigParameterSection}, {type=MovieSceneControlRigParameterSection}]
   }
  ]
 }]
display_rate {denominator=1, numerator=30}
tick_resolution {denominator=1, numerator=24000}
tracks []
